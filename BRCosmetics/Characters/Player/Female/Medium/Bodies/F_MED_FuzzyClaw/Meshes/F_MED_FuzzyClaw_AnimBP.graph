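# Post-process-style AnimBP for the FuzzyClaw female cosmetic body, animating its skirt, feathers and leg straps.
# It copies the parent mesh pose, then chains Anim Dynamics springs, bone constraints and a deformation Control Rig fed sliding and LOD flags.
# Wearing a backbling drops most back and side skirt-feather physics alpha to 0.1.
# Emotes blend in additive rotations on skirt and feather deform bones.
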AnimBlueprint /BRCosmetics/Characters/Player/Female/Medium/Bodies/F_MED_FuzzyClaw/Meshes/F_MED_FuzzyClaw_AnimBP
target_skeleton=&/F_MED_FuzzyClaw_Skeleton
graphs [
 AnimGraph {nodes [
   n1 = Anim Dynamics - Bone: dyn_skirt_f_1_r!(ComponentPose=Local To Component(LocalPose=Copy Pose From Mesh() {bUseAttachedParent=true}), Alpha=1, GravityOverride:Vector="1,1,1") -> (Pose) {
    angularDampingOverride=1
    angularSpringConstant=1500
    bAngularSpring=true
    bGravityOverrideInSimSpace=true
    bOverrideAngularDamping=true
    bOverrideLinearDamping=true
    bUseGravityOverride=true
    linearDampingOverride=0.999
    simulationSpace=RootRelative
    alphaBoolBlend {blendOption=Linear}
    boundBone {boneName=dyn_skirt_f_1_r}
    componentAppliedLinearAccClamp=<10000 10000 10000>
    componentLinearAccScale=<1.3 1.3 1.3>
    componentLinearVelScale=<2.5 2.5 2.5>
    gravityOverride=<1 1 1>
    physicsBodyDefinitions [{
      sphereCollisionRadius=0
      boundBone {boneName=dyn_skirt_f_1_r}
      boxExtents=<10 5 10>
      constraintSetup {angularLimitsMax=<5 20 -5>, angularLimitsMin=<-5 -20 -60>, angularTarget=<1 0 0>}
      localJointOffset=<0.2 0 0>
     }]
    relativeSpaceBone {boneName=root}
   }
   n2 = Anim Dynamics - Bone: dyn_skirt_f_1_l!(ComponentPose=@n1, Alpha=1, GravityOverride:Vector="1,1,1") -> (Pose) {
    angularDampingOverride=1
    angularSpringConstant=1500
    bAngularSpring=true
    bGravityOverrideInSimSpace=true
    bOverrideAngularDamping=true
    bOverrideLinearDamping=true
    bUseGravityOverride=true
    linearDampingOverride=0.999
    simulationSpace=RootRelative
    alphaBoolBlend {blendOption=Linear}
    boundBone {boneName=dyn_skirt_f_1_l}
    componentAppliedLinearAccClamp=<10000 10000 10000>
    componentLinearAccScale=<1.3 1.3 1.3>
    componentLinearVelScale=<2.5 2.5 2.5>
    gravityOverride=<1 1 1>
    physicsBodyDefinitions [{
      sphereCollisionRadius=0
      boundBone {boneName=dyn_skirt_f_1_l}
      boxExtents=<10 5 10>
      constraintSetup {angularLimitsMax=<5 20 -5>, angularLimitsMin=<-5 -20 -60>, angularTarget=<1 0 0>}
      localJointOffset=<-0.2 0 0>
     }]
    relativeSpaceBone {boneName=root}
   }
   n3 = Anim Dynamics - Bone: dyn_skirt_feather_b!(ComponentPose=@n2, Alpha=Get BackPackAlpha() -> (BackPackAlpha:double), GravityOverride:Vector="1,1,1") -> (Pose) {
    angularDampingOverride=0.9999
    angularSpringConstant=1000
    bAngularSpring=true
    bGravityOverrideInSimSpace=true
    bOverrideAngularDamping=true
    bOverrideLinearDamping=true
    bUseGravityOverride=true
    linearDampingOverride=1
    simulationSpace=RootRelative
    alphaBoolBlend {blendOption=Linear}
    boundBone {boneName=dyn_skirt_feather_b}
    componentAppliedLinearAccClamp=<200 200 200>
    componentLinearAccScale=<0.15 0.15 0.15>
    componentLinearVelScale=<0.15 0.15 0.15>
    gravityOverride=<1 1 1>
    physicsBodyDefinitions [{
      sphereCollisionRadius=0
      boundBone {boneName=dyn_skirt_feather_b}
      boxExtents=<30 5 10>
      constraintSetup {angularLimitsMax=<5 20 -5>, angularLimitsMin=<-5 -20 -60>, angularTarget=<1 0 0>}
      localJointOffset=<1 -0.1 0>
     }]
    relativeSpaceBone {boneName=root}
   }
   n4 = Anim Dynamics - Bone: dyn_skirt_b_r!(ComponentPose=@n3, Alpha=1, GravityOverride:Vector="1,1,1") -> (Pose) {
    angularDampingOverride=1
    angularSpringConstant=1500
    bAngularSpring=true
    bGravityOverrideInSimSpace=true
    bOverrideAngularDamping=true
    bOverrideLinearDamping=true
    bUseGravityOverride=true
    linearDampingOverride=0.999
    simulationSpace=RootRelative
    alphaBoolBlend {blendOption=Linear}
    boundBone {boneName=dyn_skirt_b_r}
    componentAppliedLinearAccClamp=<10000 10000 10000>
    componentLinearAccScale=<1.3 1.3 1.3>
    componentLinearVelScale=<2.5 2.5 2.5>
    gravityOverride=<1 1 1>
    physicsBodyDefinitions [{
      sphereCollisionRadius=0
      boundBone {boneName=dyn_skirt_b_r}
      boxExtents=<10 5 10>
      constraintSetup {angularLimitsMax=<5 20 0>, angularLimitsMin=<-5 -20 -50>, angularTarget=<1 0 0>}
      localJointOffset=<0.2 0 0>
     }]
    relativeSpaceBone {boneName=root}
   }
   n5 = Anim Dynamics - Bone: dyn_skirt_b!(ComponentPose=@n4, Alpha=1, GravityOverride:Vector="1,1,1") -> (Pose) {
    angularDampingOverride=1
    angularSpringConstant=1500
    bAngularSpring=true
    bGravityOverrideInSimSpace=true
    bOverrideAngularDamping=true
    bOverrideLinearDamping=true
    bUseGravityOverride=true
    linearDampingOverride=0.999
    simulationSpace=RootRelative
    alphaBoolBlend {blendOption=Linear}
    boundBone {boneName=dyn_skirt_b}
    componentAppliedLinearAccClamp=<10000 10000 10000>
    componentLinearAccScale=<1.3 1.3 1.3>
    componentLinearVelScale=<2.5 2.5 2.5>
    gravityOverride=<1 1 1>
    physicsBodyDefinitions [{
      sphereCollisionRadius=0
      boundBone {boneName=dyn_skirt_b}
      boxExtents=<10 5 10>
      constraintSetup {angularLimitsMax=<5 20 0>, angularLimitsMin=<-5 -20 -50>, angularTarget=<1 0 0>}
      localJointOffset=<0.2 0 0>
     }]
    relativeSpaceBone {boneName=root}
   }
   n6 = Anim Dynamics - Bone: dyn_skirt_b_l!(ComponentPose=@n5, Alpha=1, GravityOverride:Vector="1,1,1") -> (Pose) {
    angularDampingOverride=1
    angularSpringConstant=1500
    bAngularSpring=true
    bGravityOverrideInSimSpace=true
    bOverrideAngularDamping=true
    bOverrideLinearDamping=true
    bUseGravityOverride=true
    linearDampingOverride=0.999
    simulationSpace=RootRelative
    alphaBoolBlend {blendOption=Linear}
    boundBone {boneName=dyn_skirt_b_l}
    componentAppliedLinearAccClamp=<10000 10000 10000>
    componentLinearAccScale=<1.3 1.3 1.3>
    componentLinearVelScale=<2.5 2.5 2.5>
    gravityOverride=<1 1 1>
    physicsBodyDefinitions [{
      sphereCollisionRadius=0
      boundBone {boneName=dyn_skirt_b_l}
      boxExtents=<10 5 10>
      constraintSetup {angularLimitsMax=<5 20 0>, angularLimitsMin=<-5 -20 -50>, angularTarget=<1 0 0>}
      localJointOffset=<0.2 0 0>
     }]
    relativeSpaceBone {boneName=root}
   }
   n7 = "Constraint - deform_skirt_feather_b (P)"(ComponentPose=@n6, Alpha=1, ConstraintWeights_0=1) {
    alphaBoolBlend {blendOption=Linear}
    boneToModify {boneName=deform_skirt_feather_b}
    constraintSetup [{targetBone {boneName=dyn_skirt_b}}]
    constraintWeights [1]
   }
   n8 = Control Rig!(
    Source=Component To Local(ComponentPose=@n7)
    Alpha
    Sliding:bool=Get bIsBaseballSliding()
    Is Greater LOD1:bool=Get LOD is Greater 1() -> (LOD is Greater 1:bool)
   ) -> (Pose) {
    alphaBoolBlend {blendOption=Linear}
    controlRigAssetReference {blueprintRigClass=&/F_MED_FuzzyClaw_DeformationCtrlRig.F_MED_FuzzyClaw_DeformationCtrlRig_C}
   }
   n9 = "Transform (Modify) Bone - Bone: deform_skirt_feather_b"(ComponentPose=Local To Component(LocalPose=@n8), Alpha=1, Translation="0,-2,0", Rotation="0,0,0", Scale="1,1,1") {
    rotationMode=BMM_Additive
    rotationSpace=BCS_BoneSpace
    translationMode=BMM_Additive
    translationSpace=BCS_BoneSpace
    alphaBoolBlend {blendInTime=0.2, blendOption=Linear, blendOutTime=0.5}
    boneToModify {boneName=deform_skirt_feather_b}
    translation {y=-2}
   }
   n10 = "Constraint - deform_strap_up_l (PP)"!(ComponentPose=@n9, Alpha=1, ConstraintWeights_0=0.1332, ConstraintWeights_1:float=0.8668) -> (Pose) {
    alphaBoolBlend {blendOption=Linear}
    boneToModify {boneName=deform_strap_up_l}
    constraintSetup [{targetBone {boneName=thigh_twist_01_l}}, {targetBone {boneName=thigh_l}}]
    constraintWeights [1,1]
   }
   n11 = "Constraint - deform_strap_mid_l (PP)"!(ComponentPose=@n10, Alpha=1, ConstraintWeights_0=0.4881, ConstraintWeights_1:float=0.5119) -> (Pose) {
    alphaBoolBlend {blendOption=Linear}
    boneToModify {boneName=deform_strap_mid_l}
    constraintSetup [{targetBone {boneName=calf_l}}, {targetBone {boneName=calf_twist_02_l}}]
    constraintWeights [1,1]
   }
   n12 = "Constraint - deform_strap_down_l (PPP)"!(ComponentPose=@n11, Alpha=1, ConstraintWeights_0=0.149, ConstraintWeights_1:float=0.0009, ConstraintWeights_2:float=0.8502) -> (Pose) {
    alphaBoolBlend {blendOption=Linear}
    boneToModify {boneName=deform_strap_down_l}
    constraintSetup [{targetBone {boneName=calf_l}}, {targetBone {boneName=calf_twist_01_l}}, {targetBone {boneName=calf_twist_02_l}}]
    constraintWeights [1,1,1]
   }
   n13 = Anim Dynamics - Bone: dyn_strap_up_l(ComponentPose=@n12, Alpha=1, ExternalForce="0,0,0") {
    angularDampingOverride=0.9999
    angularSpringConstant=3500
    bAngularSpring=true
    bGravityOverrideInSimSpace=true
    bOverrideAngularDamping=true
    bOverrideLinearDamping=true
    linearDampingOverride=1
    simulationSpace=RootRelative
    alphaBoolBlend {blendOption=Linear}
    alphaScaleBiasClamp {scale=0.5}
    boundBone {boneName=dyn_strap_up_l}
    componentAppliedLinearAccClamp=<160 160 160>
    componentLinearAccScale=<0.15 0.15 0.15>
    componentLinearVelScale=<0.15 0.15 0.15>
    gravityOverride=<1 1 1>
    physicsBodyDefinitions [{
      sphereCollisionRadius=0
      boundBone {boneName=dyn_strap_up_l}
      boxExtents=<6 2 6>
      constraintSetup {angularLimitsMax=<5 5 10>, angularLimitsMin=<-5 -5 -50>, angularTarget=<1 0 0>}
      localJointOffset=<3.1 0 0>
     }]
    relativeSpaceBone {boneName=root}
   }
   n14 = Anim Dynamics - Bone: dyn_strap_mid_l(ComponentPose=@n13, Alpha=1, ExternalForce="0,0,0") {
    angularDampingOverride=0.9999
    angularSpringConstant=3000
    bAngularSpring=true
    bGravityOverrideInSimSpace=true
    bOverrideAngularDamping=true
    bOverrideLinearDamping=true
    linearDampingOverride=1
    simulationSpace=RootRelative
    alphaBoolBlend {blendOption=Linear}
    alphaScaleBiasClamp {scale=0.5}
    boundBone {boneName=dyn_strap_mid_l}
    componentAppliedLinearAccClamp=<200 200 200>
    componentLinearAccScale=<0.15 0.15 0.15>
    componentLinearVelScale=<0.15 0.15 0.15>
    gravityOverride=<1 1 1>
    physicsBodyDefinitions [{
      sphereCollisionRadius=0
      boundBone {boneName=dyn_strap_mid_l}
      boxExtents=<6 2 6>
      constraintSetup {angularLimitsMax=<5 5 0>, angularLimitsMin=<-5 -5 -50>, angularTarget=<1 0 0>}
      localJointOffset=<1.1 0 0>
     }]
    relativeSpaceBone {boneName=root}
   }
   n15 = Anim Dynamics - Bone: dyn_strap_down_l(ComponentPose=@n14, Alpha=1, ExternalForce="0,0,0") {
    angularDampingOverride=0.9999
    angularSpringConstant=3300
    bAngularSpring=true
    bGravityOverrideInSimSpace=true
    bOverrideAngularDamping=true
    bOverrideLinearDamping=true
    linearDampingOverride=1
    simulationSpace=RootRelative
    alphaBoolBlend {blendOption=Linear}
    alphaScaleBiasClamp {scale=0.5}
    boundBone {boneName=dyn_strap_down_l}
    componentAppliedLinearAccClamp=<150 150 150>
    componentLinearAccScale=<0.15 0.15 0.15>
    componentLinearVelScale=<0.15 0.15 0.15>
    gravityOverride=<1 1 1>
    physicsBodyDefinitions [{
      sphereCollisionRadius=0
      boundBone {boneName=dyn_strap_down_l}
      boxExtents=<6 2 6>
      constraintSetup {angularLimitsMax=<5 5 10>, angularLimitsMin=<-5 -5 -50>, angularTarget=<1 0.5 0>}
      localJointOffset=<2 0 0>
     }]
    relativeSpaceBone {boneName=root}
   }
   n16 = "Constraint - deform_strap_up_r (PP)"!(ComponentPose=@n15, Alpha=1, ConstraintWeights_0=0.1217, ConstraintWeights_1:float=0.8783) -> (Pose) {
    alphaBoolBlend {blendOption=Linear}
    boneToModify {boneName=deform_strap_up_r}
    constraintSetup [{targetBone {boneName=thigh_twist_01_r}}, {targetBone {boneName=thigh_r}}]
    constraintWeights [1,1]
   }
   n17 = "Constraint - deform_strap_mid_r (PP)"!(ComponentPose=@n16, Alpha=1, ConstraintWeights_0=0.4886, ConstraintWeights_1:float=0.5114) -> (Pose) {
    alphaBoolBlend {blendOption=Linear}
    boneToModify {boneName=deform_strap_mid_r}
    constraintSetup [{targetBone {boneName=calf_r}}, {targetBone {boneName=calf_twist_02_r}}]
    constraintWeights [1,1]
   }
   n18 = "Constraint - deform_strap_down_r (PPP)"!(ComponentPose=@n17, Alpha=1, ConstraintWeights_0=0.1488, ConstraintWeights_1:float=0.0009, ConstraintWeights_2:float=0.8503) -> (Pose) {
    alphaBoolBlend {blendOption=Linear}
    boneToModify {boneName=deform_strap_down_r}
    constraintSetup [{targetBone {boneName=calf_r}}, {targetBone {boneName=calf_twist_01_r}}, {targetBone {boneName=calf_twist_02_r}}]
    constraintWeights [1,1,1]
   }
   n19 = Anim Dynamics - Bone: dyn_strap_up_r(ComponentPose=@n18, Alpha=1, ExternalForce="0,0,0") {
    angularDampingOverride=0.9999
    angularSpringConstant=3500
    bAngularSpring=true
    bGravityOverrideInSimSpace=true
    bOverrideAngularDamping=true
    bOverrideLinearDamping=true
    linearDampingOverride=1
    simulationSpace=RootRelative
    alphaBoolBlend {blendOption=Linear}
    alphaScaleBiasClamp {scale=0.5}
    boundBone {boneName=dyn_strap_up_r}
    componentAppliedLinearAccClamp=<200 200 200>
    componentLinearAccScale=<0.15 0.15 0.15>
    componentLinearVelScale=<0.15 0.15 0.15>
    gravityOverride=<1 1 1>
    physicsBodyDefinitions [{
      sphereCollisionRadius=0
      boundBone {boneName=dyn_strap_up_r}
      boxExtents=<6 2 6>
      constraintSetup {angularLimitsMax=<5 5 50>, angularLimitsMin=<-5 -5 -10>, angularTarget=<1 0 0>}
      localJointOffset=<3 0 0>
     }]
    relativeSpaceBone {boneName=root}
   }
   n20 = Anim Dynamics - Bone: dyn_strap_mid_r(ComponentPose=@n19, Alpha=1, ExternalForce="0,0,0") {
    angularDampingOverride=0.9999
    angularSpringConstant=2000
    bAngularSpring=true
    bGravityOverrideInSimSpace=true
    bOverrideAngularDamping=true
    bOverrideLinearDamping=true
    linearDampingOverride=1
    simulationSpace=RootRelative
    alphaBoolBlend {blendOption=Linear}
    alphaScaleBiasClamp {scale=0.5}
    boundBone {boneName=dyn_strap_mid_r}
    componentAppliedLinearAccClamp=<150 150 150>
    componentLinearAccScale=<0.15 0.15 0.15>
    componentLinearVelScale=<0.15 0.15 0.15>
    gravityOverride=<1 1 1>
    physicsBodyDefinitions [{
      sphereCollisionRadius=0
      boundBone {boneName=dyn_strap_mid_r}
      boxExtents=<6 2 6>
      constraintSetup {angularLimitsMax=<5 5 50>, angularLimitsMin=<-5 -5 0>, angularTarget=<1 0 0>}
      localJointOffset=<0.8 0 0>
     }]
    relativeSpaceBone {boneName=root}
   }
   n21 = Anim Dynamics - Bone: dyn_strap_down_r(ComponentPose=@n20, Alpha=1, ExternalForce="0,0,0") {
    angularDampingOverride=0.9999
    angularSpringConstant=3000
    bAngularSpring=true
    bGravityOverrideInSimSpace=true
    bOverrideAngularDamping=true
    bOverrideLinearDamping=true
    linearDampingOverride=1
    simulationSpace=RootRelative
    alphaBoolBlend {blendOption=Linear}
    alphaScaleBiasClamp {scale=0.5}
    boundBone {boneName=dyn_strap_down_r}
    componentAppliedLinearAccClamp=<150 150 150>
    componentLinearAccScale=<0.15 0.15 0.15>
    componentLinearVelScale=<0.15 0.15 0.15>
    gravityOverride=<1 1 1>
    physicsBodyDefinitions [{
      sphereCollisionRadius=0
      boundBone {boneName=dyn_strap_down_r}
      boxExtents=<6 2 6>
      constraintSetup {angularLimitsMax=<5 5 50>, angularLimitsMin=<-5 -5 -10>, angularTarget=<1 0 0>}
      localJointOffset=<1 0 0>
     }]
    relativeSpaceBone {boneName=root}
   }
   n22 = Anim Dynamics - Bone: dyn_feather_collar_1_l(ComponentPose=@n21, Alpha=1, ExternalForce="0,0,0") {
    angularDampingOverride=0.9999
    angularSpringConstant=3000
    bAngularSpring=true
    bGravityOverrideInSimSpace=true
    bOverrideAngularDamping=true
    bOverrideLinearDamping=true
    linearDampingOverride=1
    simulationSpace=RootRelative
    alphaBoolBlend {blendOption=Linear}
    alphaScaleBiasClamp {scale=0.5}
    boundBone {boneName=dyn_feather_collar_1_l}
    componentAppliedLinearAccClamp=<200 200 200>
    componentLinearAccScale=<0.15 0.15 0.15>
    componentLinearVelScale=<0.15 0.15 0.15>
    gravityOverride=<1 1 1>
    physicsBodyDefinitions [{
      sphereCollisionRadius=0
      boundBone {boneName=dyn_feather_collar_1_l}
      boxExtents=<10 6 6>
      constraintSetup {angularLimitsMax=<5 5 20>, angularLimitsMin=<-5 -5 -20>, angularTarget=<1 0 0>}
      localJointOffset=<0.01 0 0>
     }]
    relativeSpaceBone {boneName=root}
   }
   n23 = Anim Dynamics - Bone: dyn_feather_collar_2_l(ComponentPose=@n22, Alpha=1, ExternalForce="0,0,0") {
    angularDampingOverride=0.9999
    angularSpringConstant=3100
    bAngularSpring=true
    bGravityOverrideInSimSpace=true
    bOverrideAngularDamping=true
    bOverrideLinearDamping=true
    linearDampingOverride=1
    simulationSpace=RootRelative
    alphaBoolBlend {blendOption=Linear}
    alphaScaleBiasClamp {scale=0.5}
    boundBone {boneName=dyn_feather_collar_2_l}
    componentAppliedLinearAccClamp=<200 200 200>
    componentLinearAccScale=<0.15 0.15 0.15>
    componentLinearVelScale=<0.15 0.15 0.15>
    gravityOverride=<1 1 1>
    physicsBodyDefinitions [{
      sphereCollisionRadius=0
      boundBone {boneName=dyn_feather_collar_2_l}
      boxExtents=<10 6 6>
      constraintSetup {angularLimitsMax=<5 5 20>, angularLimitsMin=<-5 -5 -20>, angularTarget=<1 0 0>}
      localJointOffset=<0.01 0 0>
     }]
    relativeSpaceBone {boneName=root}
   }
   n24 = Anim Dynamics - Bone: dyn_feather_collar_3_l(ComponentPose=@n23, Alpha=1, ExternalForce="0,0,0") {
    angularDampingOverride=0.9999
    angularSpringConstant=3000
    bAngularSpring=true
    bGravityOverrideInSimSpace=true
    bOverrideAngularDamping=true
    bOverrideLinearDamping=true
    linearDampingOverride=1
    simulationSpace=RootRelative
    alphaBoolBlend {blendOption=Linear}
    alphaScaleBiasClamp {scale=0.5}
    boundBone {boneName=dyn_feather_collar_3_l}
    componentAppliedLinearAccClamp=<200 200 200>
    componentLinearAccScale=<0.15 0.15 0.15>
    componentLinearVelScale=<0.15 0.15 0.15>
    gravityOverride=<1 1 1>
    physicsBodyDefinitions [{
      sphereCollisionRadius=0
      boundBone {boneName=dyn_feather_collar_3_l}
      boxExtents=<10 6 6>
      constraintSetup {angularLimitsMax=<5 5 20>, angularLimitsMin=<-5 -5 -20>, angularTarget=<1 0 0>}
      localJointOffset=<0.01 0 0>
     }]
    relativeSpaceBone {boneName=root}
   }
   n25 = Anim Dynamics - Bone: dyn_feather_collar_l(ComponentPose=@n24, Alpha=1, ExternalForce="0,0,0") {
    angularDampingOverride=0.9999
    angularSpringConstant=3000
    bAngularSpring=true
    bGravityOverrideInSimSpace=true
    bOverrideAngularDamping=true
    bOverrideLinearDamping=true
    linearDampingOverride=1
    simulationSpace=RootRelative
    alphaBoolBlend {blendOption=Linear}
    alphaScaleBiasClamp {scale=0.5}
    boundBone {boneName=dyn_feather_collar_l}
    componentAppliedLinearAccClamp=<200 200 200>
    componentLinearAccScale=<0.15 0.15 0.15>
    componentLinearVelScale=<0.15 0.15 0.15>
    gravityOverride=<1 1 1>
    physicsBodyDefinitions [{
      sphereCollisionRadius=0
      boundBone {boneName=dyn_feather_collar_l}
      boxExtents=<12 8 6>
      constraintSetup {angularLimitsMax=<5 5 20>, angularLimitsMin=<-5 -5 -20>, angularTarget=<1 0 0>}
      localJointOffset=<0.01 0 0>
     }]
    relativeSpaceBone {boneName=root}
   }
   n26 = Anim Dynamics - Bone: dyn_feather_collar_4_l(ComponentPose=@n25, Alpha=1, ExternalForce="0,0,0") {
    angularDampingOverride=0.9999
    angularSpringConstant=3000
    bAngularSpring=true
    bGravityOverrideInSimSpace=true
    bOverrideAngularDamping=true
    bOverrideLinearDamping=true
    linearDampingOverride=1
    simulationSpace=RootRelative
    alphaBoolBlend {blendOption=Linear}
    alphaScaleBiasClamp {scale=0.5}
    boundBone {boneName=dyn_feather_collar_4_l}
    componentAppliedLinearAccClamp=<200 200 200>
    componentLinearAccScale=<0.15 0.15 0.15>
    componentLinearVelScale=<0.15 0.15 0.15>
    gravityOverride=<1 1 1>
    physicsBodyDefinitions [{
      sphereCollisionRadius=0
      boundBone {boneName=dyn_feather_collar_4_l}
      boxExtents=<8 5 6>
      constraintSetup {angularLimitsMax=<5 5 20>, angularLimitsMin=<-5 -5 -20>, angularTarget=<1 0 0>}
      localJointOffset=<0.01 0 0>
     }]
    relativeSpaceBone {boneName=root}
   }
   n27 = Anim Dynamics - Bone: dyn_feather_corrar_1_r(ComponentPose=@n26, Alpha=1, ExternalForce="0,0,0") {
    angularDampingOverride=0.9999
    angularSpringConstant=2800
    bAngularSpring=true
    bGravityOverrideInSimSpace=true
    bOverrideAngularDamping=true
    bOverrideLinearDamping=true
    linearDampingOverride=1
    simulationSpace=RootRelative
    alphaBoolBlend {blendOption=Linear}
    alphaScaleBiasClamp {scale=0.5}
    boundBone {boneName=dyn_feather_corrar_1_r}
    componentAppliedLinearAccClamp=<200 200 200>
    componentLinearAccScale=<0.15 0.15 0.15>
    componentLinearVelScale=<0.15 0.15 0.15>
    gravityOverride=<1 1 1>
    physicsBodyDefinitions [{
      sphereCollisionRadius=0
      boundBone {boneName=dyn_feather_corrar_1_r}
      boxExtents=<10 6 6>
      constraintSetup {angularLimitsMax=<5 5 20>, angularLimitsMin=<-5 -5 -20>, angularTarget=<1 0 0>}
      localJointOffset=<0.01 0 0>
     }]
    relativeSpaceBone {boneName=root}
   }
   n28 = Anim Dynamics - Bone: dyn_feather_corrar_2_r(ComponentPose=@n27, Alpha=1, ExternalForce="0,0,0") {
    angularDampingOverride=0.9999
    angularSpringConstant=3000
    bAngularSpring=true
    bGravityOverrideInSimSpace=true
    bOverrideAngularDamping=true
    bOverrideLinearDamping=true
    linearDampingOverride=1
    simulationSpace=RootRelative
    alphaBoolBlend {blendOption=Linear}
    alphaScaleBiasClamp {scale=0.5}
    boundBone {boneName=dyn_feather_corrar_2_r}
    componentAppliedLinearAccClamp=<200 200 200>
    componentLinearAccScale=<0.15 0.15 0.15>
    componentLinearVelScale=<0.15 0.15 0.15>
    gravityOverride=<1 1 1>
    physicsBodyDefinitions [{
      sphereCollisionRadius=0
      boundBone {boneName=dyn_feather_corrar_2_r}
      boxExtents=<10 5 6>
      constraintSetup {angularLimitsMax=<5 5 20>, angularLimitsMin=<-5 -5 -20>, angularTarget=<1 0 0>}
      localJointOffset=<0.01 0 0>
     }]
    relativeSpaceBone {boneName=root}
   }
   n29 = Anim Dynamics - Bone: dyn_feather_corrar_3_r(ComponentPose=@n28, Alpha=1, ExternalForce="0,0,0") {
    angularDampingOverride=0.9999
    angularSpringConstant=3300
    bAngularSpring=true
    bGravityOverrideInSimSpace=true
    bOverrideAngularDamping=true
    bOverrideLinearDamping=true
    linearDampingOverride=1
    simulationSpace=RootRelative
    alphaBoolBlend {blendOption=Linear}
    alphaScaleBiasClamp {scale=0.5}
    boundBone {boneName=dyn_feather_corrar_3_r}
    componentAppliedLinearAccClamp=<200 200 200>
    componentLinearAccScale=<0.15 0.15 0.15>
    componentLinearVelScale=<0.15 0.15 0.15>
    gravityOverride=<1 1 1>
    physicsBodyDefinitions [{
      sphereCollisionRadius=0
      boundBone {boneName=dyn_feather_corrar_3_r}
      boxExtents=<10 6 6>
      constraintSetup {angularLimitsMax=<5 5 20>, angularLimitsMin=<-5 -5 -20>, angularTarget=<1 0 0>}
      localJointOffset=<0.01 0 0>
     }]
    relativeSpaceBone {boneName=root}
   }
   n30 = Anim Dynamics - Bone: dyn_feather_corrar_r(ComponentPose=@n29, Alpha=1, ExternalForce="0,0,0") {
    angularDampingOverride=0.9999
    angularSpringConstant=3000
    bAngularSpring=true
    bGravityOverrideInSimSpace=true
    bOverrideAngularDamping=true
    bOverrideLinearDamping=true
    linearDampingOverride=1
    simulationSpace=RootRelative
    alphaBoolBlend {blendOption=Linear}
    alphaScaleBiasClamp {scale=0.5}
    boundBone {boneName=dyn_feather_corrar_r}
    componentAppliedLinearAccClamp=<200 200 200>
    componentLinearAccScale=<0.15 0.15 0.15>
    componentLinearVelScale=<0.15 0.15 0.15>
    gravityOverride=<1 1 1>
    physicsBodyDefinitions [{
      sphereCollisionRadius=0
      boundBone {boneName=dyn_feather_corrar_r}
      boxExtents=<11 8 6>
      constraintSetup {angularLimitsMax=<5 5 20>, angularLimitsMin=<-5 -5 -20>, angularTarget=<1 0 0>}
      localJointOffset=<0.01 0 0>
     }]
    relativeSpaceBone {boneName=root}
   }
   n31 = Anim Dynamics - Bone: dyn_feather_corrar_4_r(ComponentPose=@n30, Alpha=1, ExternalForce="0,0,0") {
    angularDampingOverride=0.9999
    angularSpringConstant=3000
    bAngularSpring=true
    bGravityOverrideInSimSpace=true
    bOverrideAngularDamping=true
    bOverrideLinearDamping=true
    linearDampingOverride=1
    simulationSpace=RootRelative
    alphaBoolBlend {blendOption=Linear}
    alphaScaleBiasClamp {scale=0.5}
    boundBone {boneName=dyn_feather_corrar_4_r}
    componentAppliedLinearAccClamp=<200 200 200>
    componentLinearAccScale=<0.15 0.15 0.15>
    componentLinearVelScale=<0.15 0.15 0.15>
    gravityOverride=<1 1 1>
    physicsBodyDefinitions [{
      sphereCollisionRadius=0
      boundBone {boneName=dyn_feather_corrar_4_r}
      boxExtents=<9 4 6>
      constraintSetup {angularLimitsMax=<5 5 20>, angularLimitsMin=<-5 -5 -20>, angularTarget=<1 0 0>}
      localJointOffset=<0.011 0 0>
     }]
    relativeSpaceBone {boneName=root}
   }
   n32 = Get BackPackAlpha()
   n33 = Anim Dynamics - Bone: dyn_skirt_feather_b_l!(ComponentPose=@n31, Alpha=@n32, GravityOverride:Vector="1,1,1") -> (Pose) {
    angularDampingOverride=0.9999
    angularSpringConstant=1000
    bAngularSpring=true
    bGravityOverrideInSimSpace=true
    bOverrideAngularDamping=true
    bOverrideLinearDamping=true
    bUseGravityOverride=true
    linearDampingOverride=1
    simulationSpace=RootRelative
    alphaBoolBlend {blendOption=Linear}
    boundBone {boneName=dyn_skirt_feather_b_l}
    componentAppliedLinearAccClamp=<200 200 200>
    componentLinearAccScale=<0.15 0.15 0.15>
    componentLinearVelScale=<0.15 0.15 0.15>
    gravityOverride=<1 1 1>
    physicsBodyDefinitions [{
      sphereCollisionRadius=0
      boundBone {boneName=dyn_skirt_feather_b_l}
      boxExtents=<30 5 10>
      constraintSetup {angularLimitsMax=<5 10 -5>, angularLimitsMin=<-5 -10 -40>, angularTarget=<1 0 0>}
      localJointOffset=<-1 0.1 0>
     }]
    relativeSpaceBone {boneName=root}
   }
   n34 = Anim Dynamics - Bone: dyn_skirt_feather_m_l!(ComponentPose=@n33, Alpha=@n32, GravityOverride:Vector="1,1,1") -> (Pose) {
    angularDampingOverride=0.9999
    angularSpringConstant=1000
    bAngularSpring=true
    bGravityOverrideInSimSpace=true
    bOverrideAngularDamping=true
    bOverrideLinearDamping=true
    bUseGravityOverride=true
    linearDampingOverride=1
    simulationSpace=RootRelative
    alphaBoolBlend {blendOption=Linear}
    boundBone {boneName=dyn_skirt_feather_m_l}
    componentAppliedLinearAccClamp=<200 200 200>
    componentLinearAccScale=<0.15 0.15 0.15>
    componentLinearVelScale=<0.15 0.15 0.15>
    gravityOverride=<1 1 1>
    physicsBodyDefinitions [{
      sphereCollisionRadius=0
      boundBone {boneName=dyn_skirt_feather_m_l}
      boxExtents=<30 5 10>
      constraintSetup {angularLimitsMax=<5 10 -5>, angularLimitsMin=<-5 -10 -40>, angularTarget=<1 0 0>}
      localJointOffset=<-1 0.1 0>
     }]
    relativeSpaceBone {boneName=root}
   }
   n35 = Anim Dynamics - Bone: dyn_skirt_feather_b_r!(ComponentPose=@n34, Alpha=@n32, GravityOverride:Vector="1,1,1") -> (Pose) {
    angularDampingOverride=0.9999
    angularSpringConstant=1000
    bAngularSpring=true
    bGravityOverrideInSimSpace=true
    bOverrideAngularDamping=true
    bOverrideLinearDamping=true
    bUseGravityOverride=true
    linearDampingOverride=1
    simulationSpace=RootRelative
    alphaBoolBlend {blendOption=Linear}
    boundBone {boneName=dyn_skirt_feather_b_r}
    componentAppliedLinearAccClamp=<200 200 200>
    componentLinearAccScale=<0.15 0.15 0.15>
    componentLinearVelScale=<0.15 0.15 0.15>
    gravityOverride=<1 1 1>
    physicsBodyDefinitions [{
      sphereCollisionRadius=0
      boundBone {boneName=dyn_skirt_feather_b_r}
      boxExtents=<30 5 10>
      constraintSetup {angularLimitsMax=<5 10 -5>, angularLimitsMin=<-5 -10 -40>, angularTarget=<1 0 0>}
      localJointOffset=<1 -0.1 0>
     }]
    relativeSpaceBone {boneName=root}
   }
   n36 = Anim Dynamics - Bone: dyn_skirt_feather_m_r!(ComponentPose=@n35, Alpha=@n32, GravityOverride:Vector="1,1,1") -> (Pose) {
    angularDampingOverride=0.9999
    angularSpringConstant=1000
    bAngularSpring=true
    bGravityOverrideInSimSpace=true
    bOverrideAngularDamping=true
    bOverrideLinearDamping=true
    bUseGravityOverride=true
    linearDampingOverride=1
    simulationSpace=RootRelative
    alphaBoolBlend {blendOption=Linear}
    boundBone {boneName=dyn_skirt_feather_m_r}
    componentAppliedLinearAccClamp=<200 200 200>
    componentLinearAccScale=<0.15 0.15 0.15>
    componentLinearVelScale=<0.15 0.15 0.15>
    gravityOverride=<1 1 1>
    physicsBodyDefinitions [{
      sphereCollisionRadius=0
      boundBone {boneName=dyn_skirt_feather_m_r}
      boxExtents=<30 5 10>
      constraintSetup {angularLimitsMax=<5 10 -5>, angularLimitsMin=<-5 -10 -40>, angularTarget=<1 0 0>}
      localJointOffset=<1 -0.1 0>
     }]
    relativeSpaceBone {boneName=root}
   }
   n37 = Anim Dynamics - Bone: dyn_skirt_feather_r!(ComponentPose=@n36, Alpha=@n32, GravityOverride:Vector="1,1,1") -> (Pose) {
    angularDampingOverride=0.9999
    angularSpringConstant=1000
    bAngularSpring=true
    bGravityOverrideInSimSpace=true
    bOverrideAngularDamping=true
    bOverrideLinearDamping=true
    bUseGravityOverride=true
    linearDampingOverride=1
    simulationSpace=RootRelative
    alphaBoolBlend {blendOption=Linear}
    boundBone {boneName=dyn_skirt_feather_r}
    componentAppliedLinearAccClamp=<200 200 200>
    componentLinearAccScale=<0.15 0.15 0.15>
    componentLinearVelScale=<0.15 0.15 0.15>
    gravityOverride=<1 1 1>
    physicsBodyDefinitions [{
      sphereCollisionRadius=0
      boundBone {boneName=dyn_skirt_feather_r}
      boxExtents=<30 5 10>
      constraintSetup {angularLimitsMax=<5 10 -5>, angularLimitsMin=<-5 -10 -40>, angularTarget=<1 0 0>}
      localJointOffset=<1 -0.1 0>
     }]
    relativeSpaceBone {boneName=root}
   }
   n38 = Anim Dynamics - Bone: dyn_skirt_feather_f_r!(ComponentPose=@n37, Alpha=1, GravityOverride:Vector="1,1,1") -> (Pose) {
    angularDampingOverride=0.9999
    angularSpringConstant=1000
    bAngularSpring=true
    bGravityOverrideInSimSpace=true
    bOverrideAngularDamping=true
    bOverrideLinearDamping=true
    bUseGravityOverride=true
    linearDampingOverride=1
    simulationSpace=RootRelative
    alphaBoolBlend {blendOption=Linear}
    boundBone {boneName=dyn_skirt_feather_f_r}
    componentAppliedLinearAccClamp=<200 200 200>
    componentLinearAccScale=<0.15 0.15 0.15>
    componentLinearVelScale=<0.15 0.15 0.15>
    gravityOverride=<1 1 1>
    physicsBodyDefinitions [{
      sphereCollisionRadius=0
      boundBone {boneName=dyn_skirt_feather_f_r}
      boxExtents=<30 5 10>
      constraintSetup {angularLimitsMax=<5 20 -5>, angularLimitsMin=<-5 -20 -60>, angularTarget=<1 0 0>}
      localJointOffset=<1 -0.1 0>
     }]
    relativeSpaceBone {boneName=root}
   }
   n39 = Anim Dynamics - Bone: dyn_skirt_feather_l!(ComponentPose=@n38, Alpha=1, GravityOverride:Vector="1,1,1") -> (Pose) {
    angularDampingOverride=0.9999
    angularSpringConstant=1000
    bAngularSpring=true
    bGravityOverrideInSimSpace=true
    bOverrideAngularDamping=true
    bOverrideLinearDamping=true
    bUseGravityOverride=true
    linearDampingOverride=1
    simulationSpace=RootRelative
    alphaBoolBlend {blendOption=Linear}
    boundBone {boneName=dyn_skirt_feather_l}
    componentAppliedLinearAccClamp=<200 200 200>
    componentLinearAccScale=<0.15 0.15 0.15>
    componentLinearVelScale=<0.15 0.15 0.15>
    gravityOverride=<1 1 1>
    physicsBodyDefinitions [{
      sphereCollisionRadius=0
      boundBone {boneName=dyn_skirt_feather_l}
      boxExtents=<30 5 10>
      constraintSetup {angularLimitsMax=<5 10 -5>, angularLimitsMin=<-5 -10 -40>, angularTarget=<1 0 0>}
      localJointOffset=<-1 0.1 0>
     }]
    relativeSpaceBone {boneName=root}
   }
   n40 = Anim Dynamics - Bone: dyn_skirt_feather_f_l!(ComponentPose=@n39, Alpha=1, GravityOverride:Vector="1,1,1") -> (Pose) {
    angularDampingOverride=0.9999
    angularSpringConstant=1000
    bAngularSpring=true
    bGravityOverrideInSimSpace=true
    bOverrideAngularDamping=true
    bOverrideLinearDamping=true
    bUseGravityOverride=true
    linearDampingOverride=1
    simulationSpace=RootRelative
    alphaBoolBlend {blendOption=Linear}
    boundBone {boneName=dyn_skirt_feather_f_l}
    componentAppliedLinearAccClamp=<200 200 200>
    componentLinearAccScale=<0.15 0.15 0.15>
    componentLinearVelScale=<0.15 0.15 0.15>
    gravityOverride=<1 1 1>
    physicsBodyDefinitions [{
      sphereCollisionRadius=0
      boundBone {boneName=dyn_skirt_feather_f_l}
      boxExtents=<30 5 10>
      constraintSetup {angularLimitsMax=<5 20 -5>, angularLimitsMin=<-5 -20 -60>, angularTarget=<1 0 0>}
      localJointOffset=<-1 0.1 0>
     }]
    relativeSpaceBone {boneName=root}
   }
   n41 = Get bIsPlayingEmote()
   n42 = "Transform (Modify) Bone - Bone: deform_skirt_l"!(ComponentPose=@n40, bAlphaBoolEnabled:bool=@n41, Translation="0,0,0", Rotation="0,-20,0", Scale="1,1,1") -> (Pose) {
    alphaInputType=Bool
    rotationMode=BMM_Additive
    rotationSpace=BCS_BoneSpace
    translationMode=BMM_Additive
    translationSpace=BCS_BoneSpace
    alphaBoolBlend {blendInTime=0.2, blendOption=Linear, blendOutTime=0.5}
    boneToModify {boneName=deform_skirt_l}
    rotation {yaw=-20}
   }
   n43 = "Transform (Modify) Bone - Bone: deform_skirt_r"!(ComponentPose=@n42, bAlphaBoolEnabled:bool=@n41, Translation="0,0,0", Rotation="0,-20,0", Scale="1,1,1") -> (Pose) {
    alphaInputType=Bool
    rotationMode=BMM_Additive
    rotationSpace=BCS_BoneSpace
    translationMode=BMM_Additive
    translationSpace=BCS_BoneSpace
    alphaBoolBlend {blendInTime=0.2, blendOption=Linear, blendOutTime=0.5}
    boneToModify {boneName=deform_skirt_r}
    rotation {yaw=-20}
   }
   n44 = "Transform (Modify) Bone - Bone: deform_skirt_feather_l"!(ComponentPose=@n43, bAlphaBoolEnabled:bool=@n41, Translation="0,0,0", Rotation="0,-10,0", Scale="1,1,1") -> (Pose) {
    alphaInputType=Bool
    rotationMode=BMM_Additive
    rotationSpace=BCS_BoneSpace
    translationMode=BMM_Additive
    translationSpace=BCS_BoneSpace
    alphaBoolBlend {blendInTime=0.2, blendOption=Linear, blendOutTime=0.5}
    boneToModify {boneName=deform_skirt_feather_l}
    rotation {yaw=-10}
   }
   n45 = "Transform (Modify) Bone - Bone: deform_skirt_feather_r"!(ComponentPose=@n44, bAlphaBoolEnabled:bool=@n41, Translation="0,0,0", Rotation="0,-10,0", Scale="1,1,1") -> (Pose) {
    alphaInputType=Bool
    rotationMode=BMM_Additive
    rotationSpace=BCS_BoneSpace
    translationMode=BMM_Additive
    translationSpace=BCS_BoneSpace
    alphaBoolBlend {blendInTime=0.2, blendOption=Linear, blendOutTime=0.5}
    boneToModify {boneName=deform_skirt_feather_r}
    rotation {yaw=-10}
   }
   n46 = "Transform (Modify) Bone - Bone: deform_skirt_feather_m_l"!(ComponentPose=@n45, bAlphaBoolEnabled:bool=@n41, Translation="0,0,0", Rotation="0,-10,0", Scale="1,1,1") -> (Pose) {
    alphaInputType=Bool
    rotationMode=BMM_Additive
    rotationSpace=BCS_BoneSpace
    translationMode=BMM_Additive
    translationSpace=BCS_BoneSpace
    alphaBoolBlend {blendInTime=0.2, blendOption=Linear, blendOutTime=0.5}
    boneToModify {boneName=deform_skirt_feather_m_l}
    rotation {yaw=-10}
   }
   n47 = "Transform (Modify) Bone - Bone: deform_skirt_feather_m_r"!(ComponentPose=@n46, bAlphaBoolEnabled:bool=@n41, Translation="0,0,0", Rotation="0,-10,0", Scale="1,1,1") -> (Pose) {
    alphaInputType=Bool
    rotationMode=BMM_Additive
    rotationSpace=BCS_BoneSpace
    translationMode=BMM_Additive
    translationSpace=BCS_BoneSpace
    alphaBoolBlend {blendInTime=0.2, blendOption=Linear, blendOutTime=0.5}
    boneToModify {boneName=deform_skirt_feather_m_r}
    rotation {yaw=-10}
   }
   Output Pose(Result=Component To Local(ComponentPose=@n47)) {layerGroup=None}
  ]}
 EventGraph {nodes [
   Event BlueprintUpdateAnimation()
   > Set LOD is Greater 1(LOD is Greater 1:bool="integer >= integer"(A=Get CurrentLOD(), B=2)) -> (Output_Get:bool)
   > n1 = Gravity Override State Function DEPRECATED(CharacterPartAnimInstance=Self-Reference(), GameTimeInSeconds=GetGameTimeInSeconds())
   > Set Gravity Override(Gravity Override=@n1)
   Event OnCharacterCustomizationLoaded()
   n2 = CheckForBackbling()
   > n3 = Set isWearingPack(isWearingPack:bool=OR Boolean(A=@n2, B=@n2.bIsBackblingACape)) -> (Output_Get:bool)
   > n4 = Branch(Condition=@n3)
   > Set BackPackAlpha(BackPackAlpha:double=0.1) -> (Output_Get:double)
   Set BackPackAlpha(execute=@n4.else, BackPackAlpha=1)
  ]}
]
variables [
 BackPackAlpha {type=number}
 Gravity Override {type=Vector}
 isWearingPack {type=boolean}
 LOD is Greater 1 {type=boolean}
]
