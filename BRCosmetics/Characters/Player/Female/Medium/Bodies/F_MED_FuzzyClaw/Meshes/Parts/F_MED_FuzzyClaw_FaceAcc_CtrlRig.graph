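# Control Rig for the FuzzyClaw female character's face accessory that applies procedural corrections to the hood and feather collar bones.
# The hood partly follows the head and aims at the back-hair look-at bone.
# That bone's proximity to both clavicles, remapped and curve-shaped, weights a hood offset.
# Two clavicle spherical pose readers, with debug drawing on, sum to weight feather collar offsets, though the stored hierarchy lacks those collar bones.
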ControlRigBlueprint /BRCosmetics/Characters/Player/Female/Medium/Bodies/F_MED_FuzzyClaw/Meshes/Parts/F_MED_FuzzyClaw_FaceAcc_CtrlRig
graphs [
 RigVMFunctionLibrary {links [], nodes []}
 RigVMModel {nodes [
   RigUnit_BeginExecution = &RigUnit_BeginExecution()
   &RigUnit_ParentConstraint(
    AdvancedSettings=(InterpolationType=Average,RotationOrderForFilter=XZY)
    Child=(Type=Bone,Name="deform_hood")
    Parents=((Item=(Type=Bone,Name="head"),Weight=1))
    Weight=0.4
    ExecutePin=@RigUnit_BeginExecution.ExecutePin
   )
   > &RigUnit_AimConstraintLocalSpaceOffset(
    Child=(Type=Bone,Name="deform_hood")
    Filter=(bX=True,bY=True,bZ=True)
    Parents=((Item=(Type=Bone,Name="dyn_C_hairBack_Lookat"),Weight=1))
    UpAxis=(X=0,Y=-1,Z=0)
    Weight=1
    WorldUp=(Target=(X=0,Y=0,Z=1),Kind=Location,Space=(Type=Bone,Name="dyn_C_hairBack_Lookat"))
   )
   n1 = &RigUnit_GetTransform(Item=(Type=Bone,Name="dyn_C_hairBack_Lookat"), bInitial=False)
   Remap = &RigVMFunction_MathDoubleRemap(
    SourceMaximum=23
    SourceMinimum=0
    TargetMaximum=0
    TargetMinimum=1
    Value=&RigVMFunction_MathVectorDistance(
     A=&RigUnit_GetTransform(Item=(Type=Bone,Name="clavicle_r"), bInitial=False).Transform.Translation
     B=@n1.Transform.Translation
    ).Result
   )
   RigUnit_GetTransform = &RigUnit_GetTransform(Item=(Type=Bone,Name="clavicle_l"), bInitial=False)
   RigVMFunction_MathVectorDistance = &RigVMFunction_MathVectorDistance(A=@RigUnit_GetTransform.Transform.Translation, B=@n1.Transform.Translation)
   Add = &RigVMFunction_MathDoubleAdd(
    A=@Remap.Result
    B=&RigVMFunction_MathDoubleRemap(SourceMaximum=23, SourceMinimum=0, TargetMaximum=0, TargetMinimum=1, Value=@RigVMFunction_MathVectorDistance.Result).Result
   )
   Clamp = &RigVMFunction_MathDoubleClamp(Value=@Add.Result)
   RigVMFunction_AnimEvalRichCurve = &RigVMFunction_AnimEvalRichCurve(
    Curve=(EditorCurveData=(Keys=((),(Time=0.6993,Value=0.5006),(Time=0.8544,Value=0.9105),(Time=1,Value=1))))
    SourceMinimum=0
    TargetMinimum=0
    Value=@Clamp.Result
   )
   > &RigUnit_ModifyTransforms(
    ItemToModify=((Item=(Type=Bone,Name="deform_hood"),Transform=(Rotation=(X=0,Y=0,Z=0.1305,W=0.9914),Translation=(X=6,Y=1,Z=0),Scale3D=(X=1,Y=1,Z=1.2))))
    WeightMinimum=0
    Weight=@RigVMFunction_AnimEvalRichCurve.Result
   )
   > n2 = &RigUnit_SphericalPoseReader(
    ActiveRegionScaleFactors=(PositiveWidth=1,NegativeWidth=1,PositiveHeight=1,NegativeHeight=1)
    ActiveRegionSize=0.1
    Debug=(bDrawDebug=true,DebugScale=25,DebugSegments=20,DebugThickness=0.25)
    DriverAxis=(X=-1,Y=0,Z=0)
    DriverItem=(Type=Bone,Name="clavicle_r")
    FalloffRegionScaleFactors=(PositiveWidth=1,NegativeWidth=1,PositiveHeight=1,NegativeHeight=1)
    FalloffSize=0.3
    OptionalParentItem=(Type=Bone)
    RotationOffset=(X=150,Y=0,Z=50)
   )
   > n3 = &RigUnit_SphericalPoseReader(
    ActiveRegionScaleFactors=(PositiveWidth=1,NegativeWidth=1,PositiveHeight=1,NegativeHeight=1)
    ActiveRegionSize=0.1
    Debug=(bDrawDebug=true,DebugScale=25,DebugSegments=20,DebugThickness=0.25)
    DriverAxis=(X=1,Y=0,Z=0)
    DriverItem=(Type=Bone,Name="clavicle_l")
    FalloffRegionScaleFactors=(PositiveWidth=1,NegativeWidth=1,PositiveHeight=1,NegativeHeight=1)
    FalloffSize=0.3
    OptionalParentItem=(Type=Bone)
    RotationOffset=(X=10,Y=10,Z=-160)
   )
   > &RigUnit_ModifyTransforms(
    ItemToModify="((Item=(Type=Bone,Name=\"deform_feather_collar_l\"),Transform=(Rotation=(X=-0.0226,Y=-0.2578,Z=-0.0842,W=0.9623),Translation=(X=0,Y=-2,Z=2),Scale3D=(X=1,Y=1,Z=1))),(Item=(Type=Bone,Name=\"dyn_feather_collar_4_l\"),Transform=(Rotation=(X=-0.2739,Y=-0.0893,Z=0.3 ...(truncated)"
    WeightMinimum=0
    Weight=&RigVMFunction_MathDoubleAdd(A=@n2.OutputParam, B=@n3.OutputParam).Result
   )
  ]}
]
hierarchy [
 root {loc=<0 0 0>, rot=<0 0 0 1>}
 pelvis {parent=root, loc=<0 -0.5301 98.0571>, rot=<0.0127 0.707 -0.0127 -0.707>}
 spine_01 {parent=pelvis, loc=<6.1208 -0.9328 0>, rot=<0 0 0.034 -0.9994>}
 spine_02 {parent=spine_01, loc=<6.5996 -0.2095 0>, rot=<0 0 0.0238 -0.9997>}
 spine_03 {parent=spine_02, loc=<6.4914 -0.1237 0>, rot=<0 0 0.0149 -0.9999>}
 spine_04 {parent=spine_03, loc=<6.9598 0.0794 0>, rot=<0 0 -0.0132 -0.9999>}
 spine_05 {parent=spine_04, loc=<6.9987 0 0>, rot=<0 0 -0.0015 -1>}
 neck_01 {parent=spine_05, loc=<11.3138 0.1529 0>, rot=<0 0 -0.2003 -0.9797>}
 neck_02 {parent=neck_01, loc=<4.5627 0 0>, rot=<0 0 0.0504 -0.9987>}
 head {parent=neck_02, loc=<5.3914 0 0>, rot=<0 0 0.115 -0.9934>}
 dyn_C_hairBack_Lookat {parent=head, loc=<5.4438 -7.4865 0>, rot=<0.0002 -1 0 0.0002>}
 dyn_C_hairBack_0 {parent=dyn_C_hairBack_Lookat, loc=<0 0 0>, rot=<-0.7071 -0.0101 -0.0103 -0.707>}
 dyn_bangs_1 {parent=head, loc=<7.071 9.668 -0.346>, rot=<0.0041 -0.9953 0.0004 0.0966>}
 dyn_bangs_2 {parent=head, loc=<7.3387 7.7642 -5.0023>, rot=<0.2306 -0.9273 -0.2882 -0.0625>}
 dyn_R_hairFront_Lookat {parent=neck_02, loc=<2.1634 0.8877 7.74>, rot=<-0.3189 -0.9478 0.0003 0.0033>}
 dyn_R_hairFront_0 {parent=dyn_R_hairFront_Lookat, loc=<0 0 0>, rot=<0.7137 0.0152 0.0035 -0.7003>}
 dyn_R_hairFront_1 {parent=dyn_R_hairFront_0, loc=<7.8392 0 0>, rot=<0.7468 0.1389 -0.0461 -0.6488>}
 dyn_R_hairFront_2 {parent=dyn_R_hairFront_1, loc=<6.3288 0 0>, rot=<0.0998 0.0613 -0.0417 -0.9922>}
 dyn_R_hairFront_3 {parent=dyn_R_hairFront_2, loc=<6.9145 0 0>, rot=<-0.5167 -0.2077 -0.2426 -0.7944>}
 dyn_R_hairFront_4 {parent=dyn_R_hairFront_3, loc=<6.5516 0 0>, rot=<-0.0643 -0.3709 0.9194 0.114>}
 dyn_R_hairFront_5 {parent=dyn_R_hairFront_1, loc=<5.9929 0.8889 2.3659>, rot=<0.11 0.115 -0.1546 -0.9751>}
 dyn_R_hairFront_7 {parent=dyn_R_hairFront_5, loc=<6.9145 0 0>, rot=<-0.5626 -0.2436 -0.3225 -0.7212>}
 dyn_R_hairFront_6 {parent=dyn_R_hairFront_1, loc=<6.8165 -0.7238 -3.6475>, rot=<0.0361 0.0368 -0.0516 -0.9973>}
 dyn_L_hairFront_Lookat {parent=neck_02, loc=<2.163 0.8878 -7.74>, rot=<-0.3189 -0.9478 -0.0003 -0.0033>}
 dyn_L_hairFront_0 {parent=dyn_L_hairFront_Lookat, loc=<0 0 0>, rot=<0.7148 0.0272 0.0174 -0.6986>}
 dyn_L_hairFront_1 {parent=dyn_L_hairFront_0, loc=<7.5351 0 0>, rot=<-0.1706 0.0631 0.0839 -0.9797>}
 dyn_L_hairFront_2 {parent=dyn_L_hairFront_1, loc=<6.2406 0 0>, rot=<-0.1097 -0.0031 0.0898 -0.9899>}
 dyn_L_hairFront_3 {parent=dyn_L_hairFront_2, loc=<6.2138 0 0>, rot=<-0.0298 -0.3437 0.0345 -0.938>}
 dyn_L_hairFront_4 {parent=dyn_L_hairFront_3, loc=<8.1769 0 0>, rot=<0.3762 -0.0748 0.0927 -0.9188>}
 dyn_L_hairFront_5 {parent=dyn_L_hairFront_1, loc=<5.5965 -3.4629 -0.7818>, rot=<-0.1994 -0.072 0.0744 -0.9744>}
 dyn_L_hairFront_6 {parent=dyn_L_hairFront_5, loc=<5.0282 -0.4656 -1.3169>, rot=<-0.2242 -0.3126 0.0762 -0.9199>}
 dyn_L_hairFront_7 {parent=dyn_L_hairFront_1, loc=<6.7696 2.9096 0.1979>, rot=<0.01 -0.0143 0.0543 -0.9984>}
 deform_hood {parent=neck_01, loc=<-1.7133 -5.4183 0>, rot=<0 0 0.3779 -0.9258>}
 dyn_hood {parent=deform_hood, loc=<0 0 0>, rot=<0 0 0 1>}
 clavicle_l {parent=spine_05, loc=<5.4831 -0.1053 -3.2946>, rot=<0.0465 -0.7289 -0.009 -0.6829>}
 upperarm_l {parent=clavicle_l, loc=<9.8658 -1.2136 -0.4105>, rot=<0.0164 -0.3245 -0.0269 -0.9454>}
 upperarm_twist_01_l {parent=upperarm_l, loc=<7.2135 -0.1134 0>, rot=<0 0 0 1>}
 deform_upperarm_bk_base_l {parent=upperarm_twist_01_l, loc=<-6.4053 -4.3746 0.314>, rot=<0 0 0 -1>}
 deform_upperarm_bk_l {parent=deform_upperarm_bk_base_l, loc=<0 0 0>, rot=<0 0 0 1>}
 deform_upperarm_fr_base_l {parent=upperarm_twist_01_l, loc=<-5.6312 5.092 0.3354>, rot=<0 0 0 -1>}
 deform_upperarm_fr_l {parent=deform_upperarm_fr_base_l, loc=<0 0 0>, rot=<0 0 0 1>}
 deform_upperarm_up_base_l {parent=upperarm_twist_01_l, loc=<-8.891 0.4825 3.9517>, rot=<0 0 0 -1>}
 deform_upperarm_up_l {parent=deform_upperarm_up_base_l, loc=<0 0 0>, rot=<0 0 0 1>}
 clavicle_r {parent=spine_05, loc=<5.4831 -0.1053 3.2946>, rot=<0.7289 0.0465 -0.6829 0.009>}
 upperarm_r {parent=clavicle_r, loc=<-9.8658 1.2136 0.4105>, rot=<0.0164 -0.3245 -0.0269 -0.9454>}
 upperarm_twist_01_r {parent=upperarm_r, loc=<-7.2135 0.1134 0>, rot=<0 0 0 -1>}
 deform_upperarm_bk_base_r {parent=upperarm_twist_01_r, loc=<6.4053 4.3746 -0.314>, rot=<0 0 0 -1>}
 deform_upperarm_bk_r {parent=deform_upperarm_bk_base_r, loc=<0 0 0>, rot=<0 0 0 1>}
 deform_upperarm_fr_base_r {parent=upperarm_twist_01_r, loc=<5.6312 -5.092 -0.3354>, rot=<0 0 0 -1>}
 deform_upperarm_fr_r {parent=deform_upperarm_fr_base_r, loc=<0 0 0>, rot=<0 0 0 1>}
 deform_upperarm_up_base_r {parent=upperarm_twist_01_r, loc=<8.891 -0.4825 -3.9517>, rot=<0 0 0 -1>}
 deform_upperarm_up_r {parent=deform_upperarm_up_base_r, loc=<0 0 0>, rot=<0 0 0 1>}
 deform_pec_aim_base_l {parent=spine_04, loc=<3.2456 9.3309 -1.2237>, rot=<0.0603 -0.4033 0.3318 -0.8507>}
 deform_pec_aim_l {parent=deform_pec_aim_base_l, loc=<0 0 0>, rot=<0 0 0 1>}
 deform_pec_base_l {parent=deform_pec_aim_l, loc=<9.3665 0 0>, rot=<0 0 0 1>}
 deform_pec_l {parent=deform_pec_base_l, loc=<0 0 0>, rot=<0 0 0 1>}
 deform_pec_aim_base_r {parent=spine_04, loc=<3.2452 9.3309 1.2237>, rot=<-0.4033 -0.0603 0.8507 0.3318>}
 deform_pec_aim_r {parent=deform_pec_aim_base_r, loc=<0 0 0>, rot=<0 0 0 1>}
 deform_pec_base_r {parent=deform_pec_aim_r, loc=<-9.3667 -0.0003 -0.0004>, rot=<0 0 0 1>}
 deform_pec_r {parent=deform_pec_base_r, loc=<0 0 0>, rot=<0 0 0 1>}
]
