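# Physics asset for the FuzzyClaw female medium body's face accessory, simulating left and right front hair strands with capsule bodies.
# Spine, neck, head, clavicle and upper-arm capsules form the body volume around the hair, linked by free-swing, free-twist constraints.
# Each side hangs from neck_02 through a Lookat bone locked to the strand root, then continues in angle-limited chains that branch at segment 1.
# An extra left-only constraint ties segment 3 to branch segment 6, closing a loop.
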PhysicsAsset /BRCosmetics/Characters/Player/Female/Medium/Bodies/F_MED_FuzzyClaw/Meshes/Parts/F_MED_FuzzyClaw_FaceAcc_Physics
bodies {
 clavicle_l [{shapeType=Capsule, length=6.4474, radius=8.1511, shapeName=clavicle_l_capsule, center=<4.6168 -0.6207 0.8647>, rotation=<rot 90 0 0>}]
 clavicle_r [{shapeType=Capsule, length=4.2792, radius=8.1511, shapeName=clavicle_l_capsule, center=<-4.4497 1.7748 -0.2942>, rotation=<rot 90 180 0>}]
 dyn_L_hairFront_0 [{shapeType=Capsule, length=5.4064, radius=2.0936, shapeName=dyn_L_hairFront_0_capsule, center=<3.3063 0 1.0338>, rotation=<rot 90 0 0>}]
 dyn_L_hairFront_1 [{
   shapeType=Capsule
   length=4.0861
   radius=2.5
   shapeName=dyn_L_hairFront_1_capsule
   center=<3.3063 0.5691 1.6336>
   rotation=<rot 0.0003 0.0001 104.0736>
  }]
 dyn_L_hairFront_2 [{shapeType=Capsule, length=8.2144, radius=2.5, shapeName=dyn_L_hairFront_2_capsule, center=<3.3063 0.4799 1.0338>, rotation=<rot 0 0 90>}]
 dyn_L_hairFront_3 [
  {shapeType=Capsule, length=7.7483, radius=2.5, shapeName=dyn_L_hairFront_3_capsule, center=<5.8298 0.4402 0.0976>, rotation=<rot 80.8353 0 0>}
  {
   shapeType=Capsule
   length=7.2926
   radius=2
   shapeName=dyn_L_hairFront_3_capsule
   center=<2.5209 -3.277 -0.9492>
   rotation=<rot 26.8719 97.1741 82.9723>
  }
  {
   shapeType=Capsule
   length=5
   radius=2
   shapeName=dyn_L_hairFront_3_capsule
   center=<1.1413 4.4285 0.3345>
   rotation=<rot 69.8843 163.5053 158.1921>
  }
 ]
 dyn_L_hairFront_5 [{
   shapeType=Capsule
   length=5.2518
   radius=2.1006
   shapeName=dyn_L_hairFront_5_capsule
   center=<1.5477 -0.2106 0.5478>
   rotation=<rot 0 -86.4147 -80.4917>
  }]
 dyn_L_hairFront_6 [{
   shapeType=Capsule
   length=5.1023
   radius=1.9726
   shapeName=dyn_L_hairFront_6_capsule
   center=<4.6581 -0.2224 0.2123>
   rotation=<rot 0 82.4658 -88.9388>
  }]
 dyn_L_hairFront_7 [{
   shapeType=Capsule
   length=7.1413
   radius=3.1861
   shapeName=dyn_L_hairFront_7_capsule
   center=<2.6609 -0.0491 0.7117>
   rotation=<rot 0 -85.7032 -89.1061>
  }]
 dyn_L_hairFront_Lookat [{
   shapeType=Capsule
   length=9.5846
   radius=2.0936
   shapeName=dyn_L_hairFront_0_capsule
   center=<6.1476 0.2971 2.1889>
   rotation=<rot 71.9406 -179.9999 -180>
  }]
 dyn_R_hairFront_0 [{shapeType=Capsule, length=5.4064, radius=2.0936, shapeName=dyn_R_hairFront_0_capsule, center=<3.3063 0 1.0338>, rotation=<rot 90 0 0>}]
 dyn_R_hairFront_1 [{shapeType=Capsule, length=5, radius=2.5, shapeName=dyn_R_hairFront_1_capsule, center=<3.3063 -1.8149 0>, rotation=<rot 0 -0.0001 -9.4011>}]
 dyn_R_hairFront_2 [
  {shapeType=Capsule, length=5, radius=2.5, shapeName=dyn_R_hairFront_2_capsule, center=<3.3063 -1.3603 0>}
  {
   shapeType=Capsule
   length=7.6174
   radius=1.8064
   shapeName=dyn_R_hairFront_2_capsule
   center=<3.8202 -0.3425 -3.8463>
   rotation=<rot 78.9248 179.9993 -170.7794>
  }
 ]
 dyn_R_hairFront_3 [
  {
   shapeType=Capsule
   length=7.8178
   radius=2.5
   shapeName=dyn_R_hairFront_3_capsule
   center=<3.3063 0.7194 1.0338>
   rotation=<rot 68.0794 -0.0001 -17.3234>
  }
  {
   shapeType=Capsule
   length=5
   radius=2
   shapeName=dyn_R_hairFront_3_capsule
   center=<2.9286 3.494 0.3264>
   rotation=<rot 69.9818 -9.281 -1.6338>
  }
 ]
 dyn_R_hairFront_5 [{
   shapeType=Capsule
   length=7.8111
   radius=2.32
   shapeName=dyn_R_hairFront_5_capsule
   center=<3.4809 -1.1961 -0.5897>
   rotation=<rot -0.7228 -91.9241 -79.7362>
  }]
 dyn_R_hairFront_6 [{
   shapeType=Capsule
   length=8.3644
   radius=2.6424
   shapeName=dyn_R_hairFront_6_capsule
   center=<3.0207 -0.6356 0.1937>
   rotation=<rot 0 84.8911 -85.8001>
  }]
 dyn_R_hairFront_7 [{
   shapeType=Capsule
   length=4.545
   radius=2.2388
   shapeName=dyn_R_hairFront_7_capsule
   center=<3.2843 0.0311 0.5174>
   rotation=<rot 0.0026 89.6952 -91.7161>
  }]
 dyn_R_hairFront_Lookat [{shapeType=Capsule, length=5.4064, radius=2.0936, shapeName=dyn_R_hairFront_0_capsule, center=<3.3063 1.1053 0.0449>, rotation=<rot 90 0 0>}]
 head [{shapeType=Capsule, length=5.148, radius=7.4325, shapeName=head_capsule, center=<3.3063 0 0>, rotation=<rot 90 0 0>}]
 neck_01 [
  {shapeType=Capsule, length=5.1082, radius=5.5512, shapeName=neck_01_capsule, center=<3.3063 0 0>, rotation=<rot 90 0 0>}
  {
   shapeType=Capsule
   length=34.769
   radius=7.7653
   shapeName=neck_01_capsule
   center=<-2.7003 -3.6212 -0.7764>
   rotation=<rot 0 -0.0005 -8.9944>
  }
 ]
 neck_02 [{shapeType=Capsule, length=2.6678, radius=6.0799, shapeName=neck_02_capsule, center=<3.3063 0 0>, rotation=<rot 90 0 0>}]
 spine_03 [
  {shapeType=Capsule, length=0, radius=9.9065, shapeName=spine_03_capsule, center=<14.212 1.218 2.2213>}
  {shapeType=Capsule, length=16.3559, radius=13.6441, shapeName=spine_03_capsule, center=<0 0.8095 0>, rotation=<rot 90 0 0>}
 ]
 spine_04 [{shapeType=Capsule, length=6.4661, radius=9.9484, shapeName=spine_04_capsule, center=<-0.5599 4.4357 0.0545>, rotation=<rot 0 -0.0002 2.3101>}]
 spine_05 [
  {shapeType=Capsule, length=13.7919, radius=8.3505, shapeName=spine_05_capsule, center=<3.3063 -0.8824 0>}
  {shapeType=Capsule, length=39.164, radius=15.5592, shapeName=spine_05_capsule, center=<9.7682 -14.9757 0>}
 ]
 upperarm_l [{shapeType=Capsule, length=11.6577, radius=3.4628, shapeName=upperarm_l_capsule, center=<6.4584 -2.5244 1.8324>, rotation=<rot 90 0 0>}]
 upperarm_r [{shapeType=Capsule, length=16.0681, radius=4.7935, shapeName=upperarm_r_capsule, center=<-6.3177 5.7893 -1.8324>, rotation=<rot 90 0 7.63>}]
 upperarm_twist_01_l [{shapeType=Capsule, length=21.4616, radius=4.2661, shapeName=upperarm_twist_01_l_capsule, center=<12 0 0>, rotation=<rot 90 0 0>}]
 upperarm_twist_01_r [{shapeType=Capsule, length=15.7841, radius=9.9435, shapeName=upperarm_twist_01_r_capsule, center=<-12 6.2103 0>, rotation=<rot 90 0 0>}]
}
constraints [
 {
  bone1Name=spine_04
  bone2Name=spine_03
  swing1LimitDegrees=45
  swing1Motion=Free
  swing2LimitDegrees=45
  swing2Motion=Free
  twistLimitDegrees=45
  twistMotion=Free
 }
 {
  bone1Name=spine_05
  bone2Name=spine_04
  swing1LimitDegrees=45
  swing1Motion=Free
  swing2LimitDegrees=45
  swing2Motion=Free
  twistLimitDegrees=45
  twistMotion=Free
 }
 {
  bone1Name=clavicle_l
  bone2Name=spine_05
  swing1LimitDegrees=45
  swing1Motion=Free
  swing2LimitDegrees=45
  swing2Motion=Free
  twistLimitDegrees=45
  twistMotion=Free
 }
 {
  bone1Name=clavicle_r
  bone2Name=spine_05
  swing1LimitDegrees=45
  swing1Motion=Free
  swing2LimitDegrees=45
  swing2Motion=Free
  twistLimitDegrees=45
  twistMotion=Free
 }
 {
  bone1Name=upperarm_l
  bone2Name=clavicle_l
  swing1LimitDegrees=45
  swing1Motion=Free
  swing2LimitDegrees=45
  swing2Motion=Free
  twistLimitDegrees=45
  twistMotion=Free
 }
 {
  bone1Name=upperarm_r
  bone2Name=clavicle_r
  swing1LimitDegrees=45
  swing1Motion=Free
  swing2LimitDegrees=45
  swing2Motion=Free
  twistLimitDegrees=45
  twistMotion=Free
 }
 {
  bone1Name=neck_01
  bone2Name=spine_05
  swing1LimitDegrees=45
  swing1Motion=Free
  swing2LimitDegrees=45
  swing2Motion=Free
  twistLimitDegrees=45
  twistMotion=Free
 }
 {
  bone1Name=neck_02
  bone2Name=neck_01
  swing1LimitDegrees=45
  swing1Motion=Free
  swing2LimitDegrees=45
  swing2Motion=Free
  twistLimitDegrees=45
  twistMotion=Free
 }
 {bone1Name=dyn_L_hairFront_Lookat, bone2Name=neck_02, swing1LimitDegrees=30, swing2LimitDegrees=30, twistLimitDegrees=10}
 {
  bone1Name=dyn_L_hairFront_0
  bone2Name=dyn_L_hairFront_Lookat
  swing1LimitDegrees=60
  swing1Motion=Locked
  swing2LimitDegrees=25
  swing2Motion=Locked
  twistLimitDegrees=10
  twistMotion=Locked
 }
 {bone1Name=dyn_L_hairFront_1, bone2Name=dyn_L_hairFront_0, swing1LimitDegrees=30, swing2LimitDegrees=30, twistLimitDegrees=10}
 {bone1Name=dyn_L_hairFront_2, bone2Name=dyn_L_hairFront_1, swing1LimitDegrees=30, swing2LimitDegrees=30, twistLimitDegrees=10}
 {bone1Name=dyn_L_hairFront_3, bone2Name=dyn_L_hairFront_2, swing1LimitDegrees=30, swing2LimitDegrees=30, twistLimitDegrees=10}
 {bone1Name=dyn_R_hairFront_Lookat, bone2Name=neck_02, swing1LimitDegrees=30, swing2LimitDegrees=30, twistLimitDegrees=10}
 {
  bone1Name=dyn_R_hairFront_0
  bone2Name=dyn_R_hairFront_Lookat
  swing1LimitDegrees=60
  swing1Motion=Locked
  swing2LimitDegrees=25
  swing2Motion=Locked
  twistLimitDegrees=10
  twistMotion=Locked
 }
 {bone1Name=dyn_R_hairFront_1, bone2Name=dyn_R_hairFront_0, swing1LimitDegrees=30, swing2LimitDegrees=30, twistLimitDegrees=10}
 {bone1Name=dyn_R_hairFront_2, bone2Name=dyn_R_hairFront_1, swing1LimitDegrees=30, swing2LimitDegrees=30, twistLimitDegrees=10}
 {bone1Name=dyn_R_hairFront_3, bone2Name=dyn_R_hairFront_2, swing1LimitDegrees=30, swing2LimitDegrees=30, twistLimitDegrees=10}
 {
  bone1Name=head
  bone2Name=neck_02
  swing1LimitDegrees=45
  swing1Motion=Free
  swing2LimitDegrees=45
  swing2Motion=Free
  twistLimitDegrees=45
  twistMotion=Free
 }
 {bone1Name=dyn_L_hairFront_5, bone2Name=dyn_L_hairFront_1, swing1LimitDegrees=5, swing2LimitDegrees=5, twistLimitDegrees=1}
 {bone1Name=dyn_L_hairFront_6, bone2Name=dyn_L_hairFront_5, swing1LimitDegrees=25, swing2LimitDegrees=25, twistLimitDegrees=10}
 {bone1Name=dyn_L_hairFront_7, bone2Name=dyn_L_hairFront_1, swing1LimitDegrees=10, swing2LimitDegrees=25, twistLimitDegrees=10}
 {bone1Name=dyn_R_hairFront_5, bone2Name=dyn_R_hairFront_1, swing1LimitDegrees=25, swing2LimitDegrees=25, twistLimitDegrees=10}
 {bone1Name=dyn_R_hairFront_7, bone2Name=dyn_R_hairFront_5, swing1LimitDegrees=25, swing2LimitDegrees=45, twistLimitDegrees=10}
 {bone1Name=dyn_R_hairFront_6, bone2Name=dyn_R_hairFront_1, swing1LimitDegrees=25, swing2LimitDegrees=10, twistLimitDegrees=10}
 {
  bone1Name=dyn_L_hairFront_3
  bone2Name=dyn_L_hairFront_6
  swing1LimitDegrees=10
  swing2LimitDegrees=10
  twistLimitDegrees=45
  twistMotion=Free
 }
 {
  bone1Name=upperarm_twist_01_l
  bone2Name=upperarm_l
  swing1LimitDegrees=45
  swing1Motion=Free
  swing2LimitDegrees=45
  swing2Motion=Free
  twistLimitDegrees=45
  twistMotion=Free
 }
 {
  bone1Name=upperarm_twist_01_r
  bone2Name=upperarm_r
  swing1LimitDegrees=45
  swing1Motion=Free
  swing2LimitDegrees=45
  swing2Motion=Free
  twistLimitDegrees=45
  twistMotion=Free
 }
]
